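# Anim Blueprint for the Frost Galore backpack, adding procedural secondary motion to its doll, hair, bookmarks and papers.
# A linear chain of bone-relative AnimDynamics and additive bone offsets runs on the input pose, simulating papers twice.
# Doll body offsets and sims blend in while falling or moving, before a final always-on doll sim.
# The event graph derives those flags and the owner's gender and body type, shifting the pack for medium females.
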AnimBlueprint /BRCosmetics/Accessories/FORT_Backpacks/Backpack_F_LRG_FrostGalore/Meshes/F_LRG_FrostGalore_Pack_AnimBP
target_skeleton=&/F_LRG_FrostGalore_Pack_Skeleton
graphs [
 AnimGraph {nodes [
   "Transform (Modify) Bone - Bone: dyn_doll_body"!(ComponentPose, bAlphaBoolEnabled:bool=True, Translation="0.45,0,0", Rotation="0,0,0", Scale="1,1,1") -> (Pose) {
    alphaInputType=Bool
    rotationMode=BMM_Additive
    rotationSpace=BCS_BoneSpace
    translationMode=BMM_Additive
    translationSpace=BCS_BoneSpace
    alphaBoolBlend {blendInTime=0.15, blendOption=Linear, blendOutTime=0.1}
    boneToModify {boneName=dyn_doll_body}
    translation {x=0.45}
   }
   n1 = "Transform (Modify) Bone - Bone: dyn_doll_body"(
    ComponentPose=Local To Component(LocalPose=Input Pose())
    bAlphaBoolEnabled=Get isFallingNotSkydiving() -> (isFallingNotSkydiving:bool)
    Translation="3,-0.8,0"
    Rotation="0,0,0"
    Scale="1,1,1"
   ) {
    alphaInputType=Bool
    rotationMode=BMM_Additive
    rotationSpace=BCS_BoneSpace
    translationMode=BMM_Additive
    translationSpace=BCS_BoneSpace
    alphaBoolBlend {blendInTime=0.5, blendOption=Linear, blendOutTime=0.2}
    boneToModify {boneName=dyn_doll_body}
    translation=<3 -0.8>
   }
   n2 = Get isMovingNotFalling() -> (isMovingNotFalling:bool)
   n3 = "Transform (Modify) Bone - Bone: dyn_doll_body"(ComponentPose=@n1, bAlphaBoolEnabled=@n2, Translation="0.45,0,0", Rotation="0,0,0", Scale="1,1,1") {
    alphaInputType=Bool
    rotationMode=BMM_Additive
    rotationSpace=BCS_BoneSpace
    translationMode=BMM_Additive
    translationSpace=BCS_BoneSpace
    alphaBoolBlend {blendInTime=0.15, blendOption=Linear, blendOutTime=0.1}
    boneToModify {boneName=dyn_doll_body}
    translation {x=0.45}
   }
   n4 = Anim Dynamics - Bone: dyn_doll_body!(ComponentPose=@n3, bAlphaBoolEnabled:bool=@n2, ExternalForce="0,0,0") -> (Pose) {
    alphaInputType=Bool
    angularDampingOverride=1
    angularSpringConstant=20000
    bAngularSpring=true
    bOverrideAngularDamping=true
    bUseGravityOverride=true
    gravityScale=0.35
    simulationSpace=BoneRelative
    alphaBoolBlend {blendInTime=0.15, blendOption=Linear, blendOutTime=0.2}
    boundBone {boneName=dyn_doll_body}
    componentAppliedLinearAccClamp=<2000 2000 2000>
    componentLinearAccScale=<0.4 0.4 4>
    componentLinearVelScale=<0.4 0.4 0.4>
    gravityOverride {z=-980}
    physicsBodyDefinitions [{
      boundBone {boneName=dyn_doll_body}
      boxExtents=<50 60 60>
      constraintSetup {angularLimitsMax=<0 8.5 3>, angularLimitsMin=<0 -8.5 -3>, angularTarget=<1 0 0>}
      localJointOffset=<7 0 0>
     }]
   }
   n5 = Anim Dynamics - Bone: dyn_doll_body(ComponentPose=@n4, bAlphaBoolEnabled=Get isFallingNotSkydiving(), ExternalForce="0,0,0") {
    alphaInputType=Bool
    angularDampingOverride=1
    angularSpringConstant=20000
    bAngularSpring=true
    bOverrideAngularDamping=true
    bUseGravityOverride=true
    gravityScale=0.35
    simulationSpace=BoneRelative
    alphaBoolBlend {blendInTime=0.5, blendOption=Linear, blendOutTime=0.2}
    boundBone {boneName=dyn_doll_body}
    componentAppliedLinearAccClamp=<4000 4000 4000>
    componentLinearAccScale=<1 1 1>
    componentLinearVelScale=<1 1 1>
    gravityOverride {z=-980}
    physicsBodyDefinitions [{
      boundBone {boneName=dyn_doll_body}
      boxExtents=<30 40 40>
      constraintSetup {angularLimitsMax=<0 30 20>, angularLimitsMin=<0 -20 -20>, angularTarget=<1 0 0>}
      localJointOffset=<7 0 0>
     }]
   }
   n6 = Anim Dynamics - Bone: dyn_bookmarkB_01(ComponentPose=@n5, Alpha=1, ExternalForce="0,0,0") {
    angularDampingOverride=1
    angularSpringConstant=20000
    bAngularSpring=true
    bOverrideAngularDamping=true
    bUseGravityOverride=true
    gravityScale=0.35
    simulationSpace=BoneRelative
    alphaBoolBlend {blendOption=Linear}
    boundBone {boneName=dyn_bookmarkB_01}
    componentLinearAccScale=<1 1 1>
    componentLinearVelScale=<1 1 1>
    gravityOverride {z=-980}
    physicsBodyDefinitions [{
      boundBone {boneName=dyn_bookmarkB_01}
      boxExtents=<30 40 40>
      constraintSetup {angularLimitsMax=<0 30 0>, angularLimitsMin=<0 -20 0>, angularTarget=<1 0 0>}
      localJointOffset=<5 0 -3>
     }]
   }
   n7 = Anim Dynamics - Bone: dyn_bookmarkA_01(ComponentPose=@n6, Alpha=1, ExternalForce="0,0,0") {
    angularDampingOverride=1
    angularSpringConstant=20000
    bAngularSpring=true
    bOverrideAngularDamping=true
    bUseGravityOverride=true
    gravityScale=0.35
    simulationSpace=BoneRelative
    alphaBoolBlend {blendOption=Linear}
    boundBone {boneName=dyn_bookmarkA_01}
    componentLinearAccScale=<1 1 1>
    componentLinearVelScale=<1 1 1>
    gravityOverride {z=-980}
    physicsBodyDefinitions [{
      boundBone {boneName=dyn_bookmarkA_01}
      boxExtents=<30 40 40>
      constraintSetup {angularLimitsMax=<0 30 0>, angularLimitsMin=<0 -20 0>, angularTarget=<1 0 0>}
      localJointOffset=<5 0 -3>
     }]
   }
   n8 = Anim Dynamics - Bone: dyn_bookmarkC_01(ComponentPose=@n7, Alpha=1, ExternalForce="0,0,0") {
    angularDampingOverride=1
    angularSpringConstant=20000
    bAngularSpring=true
    bOverrideAngularDamping=true
    bUseGravityOverride=true
    gravityScale=0.35
    simulationSpace=BoneRelative
    alphaBoolBlend {blendOption=Linear}
    boundBone {boneName=dyn_bookmarkC_01}
    componentLinearAccScale=<1 1 1>
    componentLinearVelScale=<1 1 1>
    gravityOverride {z=-980}
    physicsBodyDefinitions [{
      boundBone {boneName=dyn_bookmarkC_01}
      boxExtents=<30 40 40>
      constraintSetup {angularLimitsMax=<0 30 0>, angularLimitsMin=<0 -20 0>, angularTarget=<1 0 0>}
      localJointOffset=<5 0 -3>
     }]
   }
   n9 = Anim Dynamics - Bone: dyn_bookmarkD_01(ComponentPose=@n8, Alpha=1, ExternalForce="0,0,0") {
    angularDampingOverride=1
    angularSpringConstant=20000
    bAngularSpring=true
    bOverrideAngularDamping=true
    bUseGravityOverride=true
    gravityScale=0.35
    simulationSpace=BoneRelative
    alphaBoolBlend {blendOption=Linear}
    boundBone {boneName=dyn_bookmarkD_01}
    componentLinearAccScale=<1 1 1>
    componentLinearVelScale=<1 1 1>
    gravityOverride {z=-980}
    physicsBodyDefinitions [{
      boundBone {boneName=dyn_bookmarkD_01}
      boxExtents=<30 40 40>
      constraintSetup {angularLimitsMax=<0 30 0>, angularLimitsMin=<0 -20 0>, angularTarget=<1 0 0>}
      localJointOffset=<5 0 -3>
     }]
   }
   n10 = Anim Dynamics - Bone: dyn_paper_01!(ComponentPose=@n9, bAlphaBoolEnabled:bool=True, ExternalForce="0,0,0") -> (Pose) {
    alphaInputType=Bool
    angularDampingOverride=1
    angularSpringConstant=20000
    bAngularSpring=true
    bOverrideAngularDamping=true
    bUseGravityOverride=true
    gravityScale=0.35
    simulationSpace=BoneRelative
    alphaBoolBlend {blendOption=Linear}
    boundBone {boneName=dyn_paper_01}
    componentAppliedLinearAccClamp=<8000 8000 200000>
    componentLinearAccScale=<0.6 0.6 0.6>
    componentLinearVelScale=<0.6 0.6 0.6>
    gravityOverride {z=-980}
    physicsBodyDefinitions [{
      boundBone {boneName=dyn_paper_01}
      boxExtents=<40 50 35>
      constraintSetup {angularLimitsMax=<0 30 0>, angularLimitsMin=<0 -30 0>, angularTarget=<1 0 0>}
      localJointOffset=<2 0 -1>
     }]
   }
   n11 = Anim Dynamics - Bone: dyn_paper_02!(ComponentPose=@n10, bAlphaBoolEnabled:bool=True, ExternalForce="0,0,0") -> (Pose) {
    alphaInputType=Bool
    angularDampingOverride=1
    angularSpringConstant=20000
    bAngularSpring=true
    bOverrideAngularDamping=true
    bUseGravityOverride=true
    gravityScale=0.35
    simulationSpace=BoneRelative
    alphaBoolBlend {blendOption=Linear}
    boundBone {boneName=dyn_paper_02}
    componentAppliedLinearAccClamp=<8000 8000 200000>
    componentLinearAccScale=<0.6 0.6 0.6>
    componentLinearVelScale=<0.6 0.6 0.6>
    gravityOverride {z=-980}
    physicsBodyDefinitions [{
      boundBone {boneName=dyn_paper_02}
      boxExtents=<40 50 35>
      constraintSetup {angularLimitsMax=<0 30 0>, angularLimitsMin=<0 -30 0>, angularTarget=<1 0 0>}
      localJointOffset=<2 0 -1>
     }]
   }
   n12 = Anim Dynamics - Bone: dyn_paper_03!(ComponentPose=@n11, bAlphaBoolEnabled:bool=True, ExternalForce="0,0,0") -> (Pose) {
    alphaInputType=Bool
    angularDampingOverride=1
    angularSpringConstant=20000
    bAngularSpring=true
    bOverrideAngularDamping=true
    bUseGravityOverride=true
    gravityScale=0.35
    simulationSpace=BoneRelative
    alphaBoolBlend {blendOption=Linear}
    boundBone {boneName=dyn_paper_03}
    componentAppliedLinearAccClamp=<8000 8000 200000>
    componentLinearAccScale=<0.6 0.6 0.6>
    componentLinearVelScale=<0.6 0.6 0.6>
    gravityOverride {z=-980}
    physicsBodyDefinitions [{
      boundBone {boneName=dyn_paper_03}
      boxExtents=<40 50 35>
      constraintSetup {angularLimitsMax=<0 30 0>, angularLimitsMin=<0 -30 0>, angularTarget=<1 0 0>}
      localJointOffset=<2 0 -1>
     }]
   }
   n13 = Anim Dynamics - Bone: dyn_paper_01(ComponentPose=@n12, bAlphaBoolEnabled=True, ExternalForce="0,0,0") {
    alphaInputType=Bool
    angularDampingOverride=1
    angularSpringConstant=20000
    bAngularSpring=true
    bOverrideAngularDamping=true
    bUseGravityOverride=true
    gravityScale=0.35
    simulationSpace=BoneRelative
    alphaBoolBlend {blendOption=Linear}
    boundBone {boneName=dyn_paper_01}
    componentAppliedLinearAccClamp=<8000 8000 200000>
    componentLinearAccScale=<0.6 0.6 0.6>
    componentLinearVelScale=<0.6 0.6 0.6>
    gravityOverride {z=-980}
    physicsBodyDefinitions [{
      boundBone {boneName=dyn_paper_01}
      boxExtents=<30 40 30>
      constraintSetup {angularLimitsMax=<0 30 0>, angularLimitsMin=<0 -30 0>, angularTarget=<1 0 0>}
      localJointOffset=<2 0 -1>
     }]
    simSpaceSettings {simSpaceAngularAlpha=1}
   }
   n14 = Anim Dynamics - Bone: dyn_paper_02(ComponentPose=@n13, bAlphaBoolEnabled=True, ExternalForce="0,0,0") {
    alphaInputType=Bool
    angularDampingOverride=1
    angularSpringConstant=20000
    bAngularSpring=true
    bOverrideAngularDamping=true
    bUseGravityOverride=true
    gravityScale=0.35
    simulationSpace=BoneRelative
    alphaBoolBlend {blendOption=Linear}
    boundBone {boneName=dyn_paper_02}
    componentAppliedLinearAccClamp=<8000 8000 200000>
    componentLinearAccScale=<0.6 0.6 0.6>
    componentLinearVelScale=<0.6 0.6 0.6>
    gravityOverride {z=-980}
    physicsBodyDefinitions [{
      boundBone {boneName=dyn_paper_02}
      boxExtents=<30 40 30>
      constraintSetup {angularLimitsMax=<0 30 0>, angularLimitsMin=<0 -30 0>, angularTarget=<1 0 0>}
      localJointOffset=<2 0 -1>
     }]
    simSpaceSettings {simSpaceAngularAlpha=1}
   }
   n15 = Anim Dynamics - Bone: dyn_paper_03(ComponentPose=@n14, bAlphaBoolEnabled=True, ExternalForce="0,0,0") {
    alphaInputType=Bool
    angularDampingOverride=1
    angularSpringConstant=20000
    bAngularSpring=true
    bOverrideAngularDamping=true
    bUseGravityOverride=true
    gravityScale=0.35
    simulationSpace=BoneRelative
    alphaBoolBlend {blendOption=Linear}
    boundBone {boneName=dyn_paper_03}
    componentAppliedLinearAccClamp=<8000 8000 200000>
    componentLinearAccScale=<0.6 0.6 0.6>
    componentLinearVelScale=<0.6 0.6 0.6>
    gravityOverride {z=-980}
    physicsBodyDefinitions [{
      boundBone {boneName=dyn_paper_03}
      boxExtents=<30 40 30>
      constraintSetup {angularLimitsMax=<0 30 0>, angularLimitsMin=<0 -30 0>, angularTarget=<1 0 0>}
      localJointOffset=<2 0 -1>
     }]
    simSpaceSettings {simSpaceAngularAlpha=1}
   }
   n16 = "Transform (Modify) Bone - Bone: dyn_paper_02"(ComponentPose=@n15, Alpha=1, Translation="0,0,0", Rotation="-10.087,0.000,-0.000", Scale="1,1,1") {
    rotationMode=BMM_Additive
    rotationSpace=BCS_BoneSpace
    translationMode=BMM_Additive
    translationSpace=BCS_BoneSpace
    alphaBoolBlend {blendOption=Linear}
    boneToModify {boneName=dyn_paper_02}
    rotation=<rot -10.087 0 0>
   }
   n17 = "Transform (Modify) Bone - Bone: dyn_paper_03"(ComponentPose=@n16, Alpha=1, Translation="0,0,0", Rotation="-10.087,0.000,-0.000", Scale="1,1,1") {
    rotationMode=BMM_Additive
    rotationSpace=BCS_BoneSpace
    translationMode=BMM_Additive
    translationSpace=BCS_BoneSpace
    alphaBoolBlend {blendOption=Linear}
    boneToModify {boneName=dyn_paper_03}
    rotation {pitch=-10.087, roll=0}
   }
   n18 = Anim Dynamics - Bone: dyn_dollHairBase(ComponentPose=@n17, Alpha=1, ExternalForce="0,0,0") {
    angularDampingOverride=1
    angularSpringConstant=20000
    bAngularSpring=true
    bOverrideAngularDamping=true
    bUseGravityOverride=true
    gravityScale=0.35
    simulationSpace=BoneRelative
    alphaBoolBlend {blendOption=Linear}
    boundBone {boneName=dyn_dollHairBase}
    componentAppliedLinearAccClamp {z=200000}
    componentLinearAccScale=<0.6 0.6 1>
    componentLinearVelScale=<0.6 0.6 1>
    gravityOverride {z=-980}
    physicsBodyDefinitions [{
      boundBone {boneName=dyn_dollHairBase}
      boxExtents=<40 50 35>
      constraintSetup {angularLimitsMax=<0 0 2>, angularLimitsMin=<0 0 -25>, angularTarget=<1 0 0>}
      localJointOffset=<-2 2 0>
     }]
   }
   n19 = Anim Dynamics - Bone: dyn_dollHair_01(ComponentPose=@n18, Alpha=1, ExternalForce="0,0,0") {
    angularDampingOverride=1
    angularSpringConstant=20000
    bAngularSpring=true
    bOverrideAngularDamping=true
    bUseGravityOverride=true
    gravityScale=0.35
    simulationSpace=BoneRelative
    alphaBoolBlend {blendOption=Linear}
    boundBone {boneName=dyn_dollHair_01}
    componentAppliedLinearAccClamp {z=200000}
    componentLinearAccScale=<0.6 0.6 1>
    componentLinearVelScale=<0.6 0.6 1>
    gravityOverride {z=-980}
    physicsBodyDefinitions [{
      boundBone {boneName=dyn_dollHair_01}
      boxExtents=<40 50 35>
      constraintSetup {angularLimitsMax=<0 20 10>, angularLimitsMin=<0 -20 -20>, angularTarget=<1 0 0>}
      localJointOffset=<0 0 -2>
     }]
   }
   n20 = Anim Dynamics - Bone: dyn_dollHair_02(ComponentPose=@n19, Alpha=1, ExternalForce="0,0,0") {
    angularDampingOverride=1
    angularSpringConstant=20000
    bAngularSpring=true
    bOverrideAngularDamping=true
    bUseGravityOverride=true
    gravityScale=0.35
    simulationSpace=BoneRelative
    alphaBoolBlend {blendOption=Linear}
    boundBone {boneName=dyn_dollHair_02}
    componentAppliedLinearAccClamp {z=200000}
    componentLinearAccScale=<0.6 0.6 1>
    componentLinearVelScale=<0.6 0.6 1>
    gravityOverride {z=-980}
    physicsBodyDefinitions [{
      boundBone {boneName=dyn_dollHair_02}
      boxExtents=<40 50 35>
      constraintSetup {angularLimitsMax=<0 20 10>, angularLimitsMin=<0 -20 -20>, angularTarget=<1 0 0>}
      localJointOffset=<0 0 2>
     }]
   }
   n21 = Anim Dynamics - Bone: dyn_dollHair_03(ComponentPose=@n20, Alpha=1, ExternalForce="0,0,0") {
    angularDampingOverride=1
    angularSpringConstant=20000
    bAngularSpring=true
    bOverrideAngularDamping=true
    bUseGravityOverride=true
    gravityScale=0.35
    simulationSpace=BoneRelative
    alphaBoolBlend {blendOption=Linear}
    boundBone {boneName=dyn_dollHair_03}
    componentAppliedLinearAccClamp {z=200000}
    componentLinearAccScale=<0.6 0.6 1>
    componentLinearVelScale=<0.6 0.6 1>
    gravityOverride {z=-980}
    physicsBodyDefinitions [{
      boundBone {boneName=dyn_dollHair_03}
      boxExtents=<40 50 35>
      constraintSetup {angularLimitsMax=<0 0 10>, angularLimitsMin=<0 0 -20>, angularTarget=<1 0 0>}
      localJointOffset=<0 2 0>
     }]
   }
   n22 = "Transform (Modify) Bone - Bone: pack"(ComponentPose=@n21, Translation="0,-0.5,0", Rotation="0,0,0") {
    rotationMode=BMM_Additive
    rotationSpace=BCS_BoneSpace
    translationMode=BMM_Additive
    translationSpace=BCS_BoneSpace
    alphaBoolBlend {blendOption=Linear}
    boneToModify {boneName=pack}
    translation {y=-0.5}
   }
   n23 = "Transform (Modify) Bone - Bone: pack"!(ComponentPose=@n22, bAlphaBoolEnabled:bool=Get isFMED() -> (isFMED:bool), Translation="0,0.85,0", Rotation="-0.000,-2.992,-0.000", Scale) -> (Pose) {
    alphaInputType=Bool
    rotationMode=BMM_Additive
    rotationSpace=BCS_BoneSpace
    translationMode=BMM_Additive
    translationSpace=BCS_BoneSpace
    alphaBoolBlend {blendOption=Linear}
    boneToModify {boneName=pack}
    rotation=<rot 0 -2.992 0>
    translation {y=0.85}
   }
   n24 = Anim Dynamics - Bone: dyn_doll_body(ComponentPose=@n23, bAlphaBoolEnabled=True, ExternalForce="0,0,0") {
    alphaInputType=Bool
    angularDampingOverride=1
    angularSpringConstant=20000
    bAngularSpring=true
    bOverrideAngularDamping=true
    bUseGravityOverride=true
    gravityScale=0.35
    simulationSpace=BoneRelative
    alphaBoolBlend {blendOption=Linear}
    boundBone {boneName=dyn_doll_body}
    componentAppliedLinearAccClamp=<4000 4000 4000>
    componentLinearAccScale=<1 1 1>
    componentLinearVelScale=<1 1 1>
    gravityOverride {z=-980}
    physicsBodyDefinitions [{
      boundBone {boneName=dyn_doll_body}
      boxExtents=<55 65 65>
      constraintSetup {angularLimitsMax=<0 30 20>, angularLimitsMin=<0 -20 -20>, angularTarget=<1 0 0>}
      localJointOffset=<4 0 0>
     }]
   }
   Output Pose(Result=Component To Local(ComponentPose=@n24)) {layerGroup=None}
  ]}
 EventGraph {nodes [
   Event BlueprintUpdateAnimation()
   > n1 = Sequence!(execute) -> (then_0, then_1, then_2:Exec)
   n2 = Get bIsSkyDiving()
   n3 = NOT Boolean(A=Get bIsInFrontEnd())
   > Set isFallingNotSkydiving(isFallingNotSkydiving:bool=AND Boolean!(A=Get bIsFalling(), B=NOT Boolean(A=@n2), C:bool=@n3) -> (ReturnValue)) -> (Output_Get:bool)
   n4 = Get bIsFalling()
   n5 = Get bIsJumping()
   > Set isMovingNotFalling(isMovingNotFalling:bool=AND Boolean!(A=Get bIsMoving2D(), B=NOT Boolean(A=@n5), C=NOT Boolean(A=@n4), D:bool=@n3) -> (ReturnValue)) -> (Output_Get:bool)
   n6 = Set backpackOwner(execute=@n1.then_1, backpackOwner:Fort Pawn Object Reference=TryGetFortPawn()) -> (Output_Get:Fort Pawn Object Reference)
   > Is Valid(InputObject=@n6)
   > n7 = Sequence!(execute) -> (then_0, then_1)
   n8 = Get backpackOwner() -> (backpackOwner:Fort Pawn Object Reference)
   > n9 = FindGender!(execute, Object:Object Reference=@n8) -> (True:Exec, False:Exec, OutGender:EFortCustomGender Enum=Invalid)
   > Set isFemale(isFemale:bool="Equal (Enum)"(A:EFortCustomGender Enum=@n9, B:EFortCustomGender Enum=Female)) -> (Output_Get:bool)
   n10 = FindBodyType!(execute=@n7.then_1, Object:Object Reference=@n8) -> (True:Exec, False:Exec, OutBodyType:EFortCustomBodyType Enum=NONE)
   > Set isMedium(isMedium:bool="Equal (Enum)"(A:EFortCustomBodyType Enum=@n10, B:EFortCustomBodyType Enum=Medium)) -> (Output_Get:bool)
   Set isFMED(execute=@n1.then_2, isFMED:bool=AND Boolean!(A=Get isFemale() -> (isFemale:bool), B=Get isMedium() -> (isMedium:bool)) -> (ReturnValue)) -> (Output_Get:bool)
   > Set isNotFE(isNotFE:bool=NOT Boolean(A=Get bIsInFrontEnd())) -> (Output_Get:bool)
   TryGetPawnOwner()
  ]}
]
variables [
 backpackOwner {type=FortPawn, default=null}
 isFallingNotSkydiving {type=boolean}
 isFemale {type=boolean}
 isFMED {type=boolean}
 isMedium {type=boolean}
 isMovingNotFalling {type=boolean}
 isNotFE {type=boolean}
]
